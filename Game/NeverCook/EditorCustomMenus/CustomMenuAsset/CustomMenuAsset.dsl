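; CustomMenuAsset describes a custom editor menu, with helper functions that resolve its parent menus, child menus and tools.
; GetAllParentNames merges entries from parent alias class defaults with the custom parent list through a map, returning parallel name and section arrays.
; Tool getters blocking-load soft class references and keep those whose class defaults are globally enabled.
; GetInfoFromSection unpacks a section struct into its name plus child menu and tool arrays.
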(fn GetAllEnabledToolClassRefs
  (params (Array :type Array%20of%20Custom%20Tool%20Definition%20Soft%20Class%20References))
  (for _array_element Array
    (bind _returnvalue (Utilities|LoadClassAssetBlocking _array_element))
    (bind _ascustom_tool_definition (Utilities|Casting|CastToCustomToolDefinitionClass _returnvalue))
    (if (.IsGloballyEnabled (Class|GetClassDefaults _ascustom_tool_definition))
      (Utilities|Array|Add (LocalVariables|Getl_returns) _ascustom_tool_definition)))
  (return (LocalVariables|Getl_returns)))

(fn GetAllEnabledToolClassRefsForSection
  (params (CustomMenuSectionStruct :type Custom%20Menu%20Section%20Struct%20Structure%20%28by%20ref%29))
  (bind (_sectionname_2_9250d3394ce1384f793161b33ea9aba7 _childmenulist_7_cf1b2c6346d87c795e6adc9c87a146bc _toollist_11_2e386dd546c24cd2421e5c9133fe5fb0) (Utilities|Struct|BreakCustomMenuSectionStruct CustomMenuSectionStruct))
  (bind _result (Utilities|Set|ToArray _toollist_11_2e386dd546c24cd2421e5c9133fe5fb0))
  (for _array_element _result
    (bind _returnvalue (Utilities|LoadClassAssetBlocking _array_element))
    (bind _ascustom_tool_definition (Utilities|Casting|CastToCustomToolDefinitionClass _returnvalue))
    (if (.IsGloballyEnabled (Class|GetClassDefaults _ascustom_tool_definition))
      (Utilities|Array|Add (LocalVariables|Getl_returns) _ascustom_tool_definition)))
  (return (LocalVariables|Getl_returns)))

(fn GetAllParentNames
  (bind _l_namesectionmap (LocalVariables|Getl_nameSectionMap))
  (Utilities|FlowControl|Sequence
    (:then_0
      (bind _result (Utilities|Set|ToArray (Variables|Default|GetParentMenuAlias)))
      (for _array_element _result
        (bind _returnvalue (Utilities|LoadClassAssetBlocking _array_element))
        (bind _ascustom_menu_parent_alias (Utilities|Casting|CastToCustomMenuAliasClass _returnvalue))
        (for _array_element_2 (Class|GetClassDefaults _ascustom_menu_parent_alias)
          (Utilities|Map|Add (LocalVariables|Getl_nameSectionMap) (Utilities|Struct|BreakCustomMenuAliasStruct _array_element_2) (.SectionName_16_9250D3394CE1384F793161B33EA9ABA7 (Utilities|Struct|BreakCustomMenuAliasStruct _array_element_2))))))
    (:then_1
      (for _array_element_3 (Variables|Default|GetCustomParentMenuList)
        (Utilities|Map|Add (LocalVariables|Getl_nameSectionMap) (Utilities|Struct|BreakCustomMenuAliasStruct _array_element_3) (.SectionName_16_9250D3394CE1384F793161B33EA9ABA7 (Utilities|Struct|BreakCustomMenuAliasStruct _array_element_3)))))
    (:then_2
      (bind _keys (Utilities|Map|Keys _l_namesectionmap))
      (LocalVariables|Setl_names _keys)
      (bind _values (Utilities|Map|Values _l_namesectionmap))
      (LocalVariables|Setl_sections _values))
    (:then_3
      (return (LocalVariables|Getl_names) (LocalVariables|Getl_sections)))))

(fn GetInfoFromSection
  (params (CustomMenuSectionStruct :type Custom%20Menu%20Section%20Struct%20Structure%20%28by%20ref%29))
  (bind (_sectionname_2_9250d3394ce1384f793161b33ea9aba7 _childmenulist_7_cf1b2c6346d87c795e6adc9c87a146bc _toollist_11_2e386dd546c24cd2421e5c9133fe5fb0) (Utilities|Struct|BreakCustomMenuSectionStruct CustomMenuSectionStruct))
  (bind _result (Utilities|Set|ToArray _childmenulist_7_cf1b2c6346d87c795e6adc9c87a146bc))
  (bind _result_2 (Utilities|Set|ToArray _toollist_11_2e386dd546c24cd2421e5c9133fe5fb0))
  (return _sectionname_2_9250d3394ce1384f793161b33ea9aba7 _result _result_2))
